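; Component that adds the ad-hoc squad Join and Leave emotes to an emote category for temporary teams, on the server only.
; BeginPlay, registry updates and visibility changes each run the same rebuild, removing previously cached emote bindings.
; If ShouldShowTemporaryTeamsEmotes passes, it adds the category, creates emotes from Data Registry rows and caches the bindings.
; It then pushes emotes to players in the spatial interface, repeating after a five-second delay.
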(event EventBeginPlay
  (bind _joinsquadrow (Variables|Default|GetJoinSquadRow))
  (bind _value_2_8710cf434b488938487c73b91e9701de (Utilities|Struct|BreakDataTableRowString _joinsquadrow))
  (bind _leavesquadrow (Variables|Default|GetLeaveSquadRow))
  (bind _value_2_8710cf434b488938487c73b91e9701de_2 (Utilities|Struct|BreakDataTableRowString _leavesquadrow))
  (bind _category_name (Variables|Default|GetCategory%20Name))
  (bind _category_name_2 (Variables|Default|GetCategory%20Name))
  (Networking|SwitchHasAuthority
    (:Authority
      (Utilities|FlowControl|Sequence
        (:then_0
          (if (not (Utilities|String|EqualExactly(String) (Variables|Default|GetCachedJoinSquad) "None"))
            (CallFunction|RemoveEmotefromCategory (Variables|Default|GetOutJoinEmoteBinding) _category_name)))
        (:then_1
          (if (not (Utilities|String|EqualExactly(String) (Variables|Default|GetCachedLeaveSquad) "None"))
            (CallFunction|RemoveEmotefromCategory (Variables|Default|GetOutLeaveSquadEmoteBinding) _category_name)))
        (:then_2
          (if (CallFunction|ShouldShowTemporaryTeamsEmotes)
            (CallFunction|AddUniqueEmoteCategory (Variables|Default|GetCategory%20Name) (Utilities|Text|ToText(Name) (Variables|Default|GetCategory%20Name)) (Variables|Default|GetCategory%20Title%20Multipage))
            (Utilities|FlowControl|Sequence
              (:then_0
                (bind _returnvalue (DataRegistry|GetDataRegistryItem "FNE.TemporaryTeams.Emotes:JoinSquad" (Variables|Default|GetJoinSquadRow)))
                (if _returnvalue
                  (bind (_outcreatedemote _returnvalue_2) (CallFunction|AddEmotetoCategory _value_2_8710cf434b488938487c73b91e9701de _category_name_2))
                  (if _returnvalue_2
                    (Variables|Default|SetOutJoinEmoteBinding _outcreatedemote)
                    (Variables|Default|SetCachedJoinSquad _value_2_8710cf434b488938487c73b91e9701de))))
              (:then_1
                (bind _returnvalue_3 (DataRegistry|GetDataRegistryItem "FNE.TemporaryTeams.Emotes:LeaveSquad" (Variables|Default|GetLeaveSquadRow)))
                (if _returnvalue_3
                  (bind (_outcreatedemote_2 _returnvalue_4) (CallFunction|AddEmotetoCategory _value_2_8710cf434b488938487c73b91e9701de_2 _category_name_2))
                  (if _returnvalue_4
                    (Variables|Default|SetOutLeaveSquadEmoteBinding _outcreatedemote_2)
                    (Variables|Default|SetCachedLeaveSquad _value_2_8710cf434b488938487c73b91e9701de_2))))
              (:then_2
                (CallFunction|AddEmotestoPlayersinSpatialInterface)
                (Utilities|FlowControl|Delay 5.0)
                (CallFunction|AddEmotestoPlayersinSpatialInterface)))))))))

(event EventOnEmoteDataRegistyUpdated
  (bind _joinsquadrow (Variables|Default|GetJoinSquadRow))
  (bind _value_2_8710cf434b488938487c73b91e9701de (Utilities|Struct|BreakDataTableRowString _joinsquadrow))
  (bind _leavesquadrow (Variables|Default|GetLeaveSquadRow))
  (bind _value_2_8710cf434b488938487c73b91e9701de_2 (Utilities|Struct|BreakDataTableRowString _leavesquadrow))
  (bind _category_name (Variables|Default|GetCategory%20Name))
  (bind _category_name_2 (Variables|Default|GetCategory%20Name))
  (Networking|SwitchHasAuthority
    (:Authority
      (Utilities|FlowControl|Sequence
        (:then_0
          (if (not (Utilities|String|EqualExactly(String) (Variables|Default|GetCachedJoinSquad) "None"))
            (CallFunction|RemoveEmotefromCategory (Variables|Default|GetOutJoinEmoteBinding) _category_name)))
        (:then_1
          (if (not (Utilities|String|EqualExactly(String) (Variables|Default|GetCachedLeaveSquad) "None"))
            (CallFunction|RemoveEmotefromCategory (Variables|Default|GetOutLeaveSquadEmoteBinding) _category_name)))
        (:then_2
          (if (CallFunction|ShouldShowTemporaryTeamsEmotes)
            (CallFunction|AddUniqueEmoteCategory (Variables|Default|GetCategory%20Name) (Utilities|Text|ToText(Name) (Variables|Default|GetCategory%20Name)) (Variables|Default|GetCategory%20Title%20Multipage))
            (Utilities|FlowControl|Sequence
              (:then_0
                (bind _returnvalue (DataRegistry|GetDataRegistryItem "FNE.TemporaryTeams.Emotes:JoinSquad" (Variables|Default|GetJoinSquadRow)))
                (if _returnvalue
                  (bind (_outcreatedemote _returnvalue_2) (CallFunction|AddEmotetoCategory _value_2_8710cf434b488938487c73b91e9701de _category_name_2))
                  (if _returnvalue_2
                    (Variables|Default|SetOutJoinEmoteBinding _outcreatedemote)
                    (Variables|Default|SetCachedJoinSquad _value_2_8710cf434b488938487c73b91e9701de))))
              (:then_1
                (bind _returnvalue_3 (DataRegistry|GetDataRegistryItem "FNE.TemporaryTeams.Emotes:LeaveSquad" (Variables|Default|GetLeaveSquadRow)))
                (if _returnvalue_3
                  (bind (_outcreatedemote_2 _returnvalue_4) (CallFunction|AddEmotetoCategory _value_2_8710cf434b488938487c73b91e9701de_2 _category_name_2))
                  (if _returnvalue_4
                    (Variables|Default|SetOutLeaveSquadEmoteBinding _outcreatedemote_2)
                    (Variables|Default|SetCachedLeaveSquad _value_2_8710cf434b488938487c73b91e9701de_2))))
              (:then_2
                (CallFunction|AddEmotestoPlayersinSpatialInterface)
                (Utilities|FlowControl|Delay 5.0)
                (CallFunction|AddEmotestoPlayersinSpatialInterface)))))))))

(event EventOnTemporaryTeamsEmotesVisibilityChanged
  (bind _joinsquadrow (Variables|Default|GetJoinSquadRow))
  (bind _value_2_8710cf434b488938487c73b91e9701de (Utilities|Struct|BreakDataTableRowString _joinsquadrow))
  (bind _leavesquadrow (Variables|Default|GetLeaveSquadRow))
  (bind _value_2_8710cf434b488938487c73b91e9701de_2 (Utilities|Struct|BreakDataTableRowString _leavesquadrow))
  (bind _category_name (Variables|Default|GetCategory%20Name))
  (bind _category_name_2 (Variables|Default|GetCategory%20Name))
  (Networking|SwitchHasAuthority
    (:Authority
      (Utilities|FlowControl|Sequence
        (:then_0
          (if (not (Utilities|String|EqualExactly(String) (Variables|Default|GetCachedJoinSquad) "None"))
            (CallFunction|RemoveEmotefromCategory (Variables|Default|GetOutJoinEmoteBinding) _category_name)))
        (:then_1
          (if (not (Utilities|String|EqualExactly(String) (Variables|Default|GetCachedLeaveSquad) "None"))
            (CallFunction|RemoveEmotefromCategory (Variables|Default|GetOutLeaveSquadEmoteBinding) _category_name)))
        (:then_2
          (if (CallFunction|ShouldShowTemporaryTeamsEmotes)
            (CallFunction|AddUniqueEmoteCategory (Variables|Default|GetCategory%20Name) (Utilities|Text|ToText(Name) (Variables|Default|GetCategory%20Name)) (Variables|Default|GetCategory%20Title%20Multipage))
            (Utilities|FlowControl|Sequence
              (:then_0
                (bind _returnvalue (DataRegistry|GetDataRegistryItem "FNE.TemporaryTeams.Emotes:JoinSquad" (Variables|Default|GetJoinSquadRow)))
                (if _returnvalue
                  (bind (_outcreatedemote _returnvalue_2) (CallFunction|AddEmotetoCategory _value_2_8710cf434b488938487c73b91e9701de _category_name_2))
                  (if _returnvalue_2
                    (Variables|Default|SetOutJoinEmoteBinding _outcreatedemote)
                    (Variables|Default|SetCachedJoinSquad _value_2_8710cf434b488938487c73b91e9701de))))
              (:then_1
                (bind _returnvalue_3 (DataRegistry|GetDataRegistryItem "FNE.TemporaryTeams.Emotes:LeaveSquad" (Variables|Default|GetLeaveSquadRow)))
                (if _returnvalue_3
                  (bind (_outcreatedemote_2 _returnvalue_4) (CallFunction|AddEmotetoCategory _value_2_8710cf434b488938487c73b91e9701de_2 _category_name_2))
                  (if _returnvalue_4
                    (Variables|Default|SetOutLeaveSquadEmoteBinding _outcreatedemote_2)
                    (Variables|Default|SetCachedLeaveSquad _value_2_8710cf434b488938487c73b91e9701de_2))))
              (:then_2
                (CallFunction|AddEmotestoPlayersinSpatialInterface)
                (Utilities|FlowControl|Delay 5.0)
                (CallFunction|AddEmotestoPlayersinSpatialInterface)))))))))
